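# UI-domain material for the timeline widget of the PerfCsvPlaybackTool editor utility, drawing a horizontal progress fill.
# An If node compares the U coordinate to the saturated percent parameter, choosing FG_Color up to that point and BG_Color beyond it.
# A Texture parameter is lerped over the bar with its own RGB as alpha, so bright texels cover the fill and black texels reveal it.
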Material /Game/NeverCook/EditorUtilities/PerfTools/PerfCsvPlaybackTool/SubWidgets/PCPT_Timeline_UIMat
graph {
 properties {materialDomain=MD_UI, pixelDepthOffsetMode=PDOM_Legacy, shadingModel=MSM_Unlit, usedShadingModels=Unlit}
 nodes [
  n1 = TextureCoordinate()
  n2 = ScalarParameter() {defaultValue=0.25, parameterName=percent}
  n3 = VectorParameter() {parameterName=FG_Color, defaultValue=<rgba 0.101 0.2271 0.6615 1>}
  n4 = TextureSampleParameter2D() {parameterName=Texture, texture=/Game/Global/Textures/Default/Blanks/T_Blank_LegalWhite_sRGBColor}
  LinearInterpolate(
   A=If(
    A=ComponentMask(None=@n1) {r=true}
    B=Saturate(None=@n2)
    "A > B"=VectorParameter() {parameterName=BG_Color, defaultValue=<rgba 0.01 0.01 0.01 1>}.RGB
    "A == B"=@n3.RGB
    "A < B"=@n3.RGB
   )
   B=@n4.RGB
   Alpha=@n4.RGB
  )
 ]
}
