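# Timeline widget for the editor Perf CSV Playback tool, plotting a data series as a scrubbable graph.
# setGraphData optionally window-averages and normalizes values; DrawRT draws colored lines, dots and a threshold line into a render target, rebuilt after debounced resizes.
# The image material shows that texture plus a "percent" playhead, which mouse dragging sets, firing onValueChanged; hovering shows raw values as tooltips.
# OnPaint contains a disconnected copy of the drawing logic.
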WidgetBlueprint /Game/NeverCook/EditorUtilities/PerfTools/PerfCsvPlaybackTool/SubWidgets/PCPT_Timeline_Widget
graphs [
 DrawRT {nodes [
   DrawRT() {is=title K2Node_FunctionEntry}
   n1 = Get CurrentRT() -> (CurrentRT:Texture Render Target 2D Object Reference)
   > Is Valid(InputObject=@n1)
   > n2 = BeginDrawCanvasToRenderTarget!(execute, TextureRenderTarget:Texture Render Target 2D Object Reference=@n1) -> (then, Canvas:Canvas Object Reference, Size:Vector 2D Structure, Context:Draw To Render Target Context Structure)
   > Set l_canvas(l_canvas:Canvas Object Reference=@n2) -> (Output_Get:Canvas Object Reference)
   > n3 = Sequence()
   n4 = Length(TargetArray:"Array of Float (double-precision)s"=Get data() -> (data:"Array of Float (double-precision)s"))
   > Branch(Condition="integer > integer"(A=@n4, B=1) {is=title K2Node_CallFunction})
   > Set l_count(l_count:int=int - int(A=@n4, B=1) {is=title K2Node_CallFunction}) -> (Output_Get:int=0)
   > Set l_size(l_size:Vector 2D Structure=vector2d - vector2d!(
     A:Vector 2D Structure=Get CurrentRTSize() -> (CurrentRTSize:Vector 2D Structure)
     B:Vector 2D Structure=Get drawOffset() -> (drawOffset:Vector 2D Structure)
    ) -> (ReturnValue:Vector 2D Structure)) -> (Output_Get:Vector 2D Structure)
   > n5 = Sequence()
   > n6 = For Each Loop(Array:"Array of Float (double-precision)s"=Get data()) -> (Array Element:double)
   > Set l_curIndex(l_curIndex:int=@n6.Array Index) -> (Output_Get:int=0)
   > n7 = getDrawColorFromIndex!(execute, self, index:int=0) -> (then, ReturnValue:Linear Color Structure)
   > Set l_drawColor(l_drawColor:Linear Color Structure=@n7) -> (Output_Get:Linear Color Structure)
   > Branch(Condition="integer > integer"(A=Get l_curIndex() -> (l_curIndex:int=0), B=0))
   n8 = "To Float (Integer)"(InInt=Get l_count() -> (l_count:int=0))
   n9 = Get l_curIndex() -> (l_curIndex=0)
   n10 = "To Float (Integer)"(InInt=@n9)
   n11 = Get data()
   n12 = vector2d * vector2d(
    A:Vector 2D Structure=MakeVector2D(
     X=float / float(A=@n10, B=@n8) {is=title K2Node_CallFunction}
     Y="Get (a copy)"(Array:"Array of Float (double-precision)s"=@n11, Dimension 1=@n9) -> (Output:double)
    )
    B:Vector 2D Structure=Get l_size() -> (l_size:Vector 2D Structure)
   ) -> (ReturnValue:Vector 2D Structure) {is=title K2Node_CommutativeAssociativeBinaryOperator}
   n13 = int - int(A=Get l_curIndex() -> (l_curIndex=0), B=1)
   n14 = float / float(A="To Float (Integer)"(InInt=@n13), B="To Float (Integer)"(InInt=Get l_count() -> (l_count=0)))
   n15 = "Get (a copy)"(Array=Get data(), Dimension 1=@n13)
   n16 = Get l_canvas() -> (l_canvas:Canvas Object Reference)
   n17 = Get l_drawColor() -> (l_drawColor:Linear Color Structure)
   > Draw Line!(
    execute
    self:Canvas Object Reference=@n16
    ScreenPositionA:Vector 2D Structure=vector2d * vector2d(A=MakeVector2D(X=@n14, Y=@n15), B=Get l_size())
    ScreenPositionB:Vector 2D Structure=@n12
    Thickness:float=1
    RenderColor:Linear Color Structure=@n17
   ) -> (then)
   n18 = MakeVector2D(X=3, Y=3)
   > Draw Box!(
    execute
    self:Canvas Object Reference=@n16
    ScreenPosition:Vector 2D Structure=vector2d - vector2d(A=@n12, B=vector2d / float!(A:Vector 2D Structure=@n18, B:double=2) -> (ReturnValue:Vector 2D Structure))
    ScreenSize:Vector 2D Structure=@n18
    Thickness:float=1
    RenderColor:Linear Color Structure=@n17
   ) -> (then)
   Branch(
    execute=@n5.then_1
    Condition="Not Equal (Float)"!(A:double=Get currentTheshold() -> (currentTheshold:double), B:double=0) -> (ReturnValue:bool) {is=title K2Node_CallFunction}
   )
   n19 = Get data_Highest() -> (data_Highest:double)
   n20 = Get data_Lowest() -> (data_Lowest:double)
   n21 = Get currentTheshold()
   n22 = Get!() -> (drawOffset_X:double, drawOffset_Y:double)
   n23 = Lerp(
    A=@n22.drawOffset_Y
    B=float - float(B=@n22.drawOffset_Y) {is=title K2Node_CallFunction}
    Alpha=float - float(B=NormalizeToRange!(Value:double=@n21, RangeMin:double=@n20, RangeMax:double=@n19) -> (ReturnValue:double))
   )
   n24 = Get!() -> (l_size_X:double, l_size_Y:double)
   n25 = Get()
   n26 = float + float(
    A=float * float(A=Get!() -> (drawOffset_X:double, drawOffset_Y:double).drawOffset_Y, B=0.5)
    B="Clamp (Float)"(Value=float * float(A=@n24.l_size_Y, B=@n23), Min=0, Max=@n25.l_size_Y)
   ) {is=title K2Node_CommutativeAssociativeBinaryOperator}
   > Draw Line!(
    execute
    self=Get l_canvas()
    ScreenPositionA_X:double=0
    ScreenPositionA_Y:double=@n26
    ScreenPositionB_X:double=@n25
    ScreenPositionB_Y:double=@n26
    Thickness=1
    RenderColor=(R=1,G=1,B=1,A=1)
   ) -> (then)
   EndDrawCanvasToRenderTarget!(execute=@n3.then_1, Context:"Draw To Render Target Context Structure (by ref)"=@n2.Context) -> (then)
   n27 = Get CurrentRT()
   MakeVector2D(
    X="To Float (Integer)"(InInt=Get SizeX!(self:Texture Render Target 2D Object Reference=@n27) -> (SizeX:int=0))
    Y="To Float (Integer)"(InInt=Get SizeY!(self:Texture Render Target 2D Object Reference=@n27) -> (SizeY:int=0))
   )
  ]}
 EventGraph {nodes [
   n1 = Clear and Invalidate Timer by Handle(Handle=Get TimerHandle_ValidateRT() -> (TimerHandle_ValidateRT:Timer Handle Structure))
   n2 = Set Timer by Event(execute=[@n1.then, @n5.else], Delegate=Create Event(), Time=1, bLooping=false, bMaxOncePerFrame=true)
   > Set TimerHandle_ValidateRT(TimerHandle_ValidateRT:Timer Handle Structure=@n2) -> (Output_Get:Timer Handle Structure)
   n3 = Event Tick()
   > OnTickValidateVisuals()
   n4 = GetLocalSize!(Geometry:"Geometry Structure (by ref)"=@n3.MyGeometry) -> (ReturnValue:Vector 2D Structure)
   > Branch(Condition="Not Equal (Vector2D)"!(A:Vector 2D Structure=@n4, B:Vector 2D Structure=Get LastDrawSize() -> (LastDrawSize:Vector 2D Structure), ErrorTolerance:float) -> (ReturnValue:bool=~) {is=title K2Node_PromotableOperator})
   > Set LastDrawSize(LastDrawSize:Vector 2D Structure=@n4) -> (Output_Get:Vector 2D Structure)
   > Is Valid(InputObject=Get CurrentRT())
   > n5 = Branch(Condition=Is Timer Active by Handle!(Handle:Timer Handle Structure=Get TimerHandle_ValidateRT()) -> (ReturnValue:bool))
   Event OnMouseEnter()
   > Set isMouseOver(isMouseOver:bool=true) -> (Output_Get:bool)
   Event Destruct()
   > Is Valid(InputObject=Get CurrentRT())
   > ReleaseRenderTarget2D!(execute, TextureRenderTarget:Texture Render Target 2D Object Reference=Get CurrentRT()) -> (then)
   > Set CurrentRT(CurrentRT:Texture Render Target 2D Object Reference) -> (Output_Get:Texture Render Target 2D Object Reference)
   Branch(Condition=Event PreConstruct().IsDesignTime)
   > setValue!(execute, self, value:double=Get value() -> (value:double), throwEvent:bool=true) -> (then)
   > Set validateNextTick(validateNextTick:bool=true) -> (Output_Get:bool)
  ]}
 GetTimelineMaterial {nodes [
   GetTimelineMaterial() {is=title K2Node_FunctionEntry}
   > n1 = GetDynamicMaterial(self=Get Image_80() -> (Image_80:Image Object Reference))
   > Return Node!(execute, ReturnValue:Material Instance Dynamic Object Reference=@n1)
  ]}
 GetTooltipText {nodes [
   GetTooltipText() {is=title K2Node_FunctionEntry}
   > n1 = GetMousePositionOnPlatform!(execute) -> (then, ReturnValue:Vector 2D Structure)
   n2 = GetCachedGeometry!(self:Widget Object Reference) -> (ReturnValue:"Geometry Structure (by ref)")
   n3 = GetLocalSize!(Geometry=@n2) -> (ReturnValue_X:double, ReturnValue_Y:double)
   n4 = "Absolute to Local (Coordinates)"!(Geometry:"Geometry Structure (by ref)"=@n2, AbsoluteCoordinate:Vector 2D Structure=@n1) -> (ReturnValue_X:double, ReturnValue_Y:double)
   n5 = Get data()
   n6 = Lerp(
    A=0
    B="To Float (Integer)"(InInt=Last Index!(TargetArray:"Array of Float (double-precision)s"=@n5) -> (ReturnValue=0))
    Alpha="Clamp (Float)"(Value=float / float(A=@n4, B=@n3), Min=0, Max=1)
   )
   n7 = "Get (a copy)"(Array=Get InputData() -> (InputData:"Array of Float (double-precision)s"), Dimension 1=Round(A=@n6) {is=title K2Node_CallFunction})
   > SetToolTipText!(
    execute
    self:Widget Object Reference
    InToolTipText:"Text (by ref)"="To Text (Float)"(Value=@n7, MinimumIntegralDigits=1, MaximumIntegralDigits=324, MinimumFractionalDigits=0, MaximumFractionalDigits=3)
   ) -> (then)
   Append!(A="To String (Float)"(InDouble=@n6), B=" == ", C:String="To String (Float)"(InDouble=@n7)) -> (ReturnValue)
  ]}
 OnMouseButtonDown {nodes [
   n1 = OnMouseButtonDown!() -> (then, MyGeometry:Geometry Structure, MouseEvent:"Pointer Event Structure (by ref)") {is=title K2Node_FunctionEntry}
   > Branch(Condition=Is Mouse Button Down!(Input:"Pointer Event Structure (by ref)"=@n1.MouseEvent, MouseButton:Key Structure=LeftMouseButton) -> (ReturnValue:bool))
   > calculateValueFromMouseData()
   > Set isDragging(isDragging:bool=true) -> (Output_Get:bool)
   > Return Node(ReturnValue:Event Reply Structure=CaptureMouse!(Reply:"Event Reply Structure (by ref)"=Unhandled(), CapturingWidget:Widget Object Reference=Self-Reference()) -> (ReturnValue:Event Reply Structure))
  ]}
 OnMouseButtonUp {nodes [
   OnMouseButtonUp!() -> (then, MyGeometry:Geometry Structure, MouseEvent:"Pointer Event Structure (by ref)") {is=title K2Node_FunctionEntry}
   > Set isDragging(isDragging=false)
   > Return Node(ReturnValue=ReleaseMouseCapture!(Reply:"Event Reply Structure (by ref)"=Handled()) -> (ReturnValue:Event Reply Structure))
  ]}
 OnMouseMove {nodes [
   n1 = OnMouseMove!() -> (then, MyGeometry:Geometry Structure, MouseEvent:"Pointer Event Structure (by ref)") {is=title K2Node_FunctionEntry}
   > GetTooltipText() {is=title K2Node_CallFunction}
   > Branch(Condition=Get isDragging() -> (isDragging:bool))
   > handleMouseMove!(execute, self, Geometry:"Geometry Structure (by ref)"=@n1, Input:"Pointer Event Structure (by ref)"=@n1.MouseEvent) -> (then)
   > Return Node(ReturnValue=Handled())
  ]}
 OnPaint {nodes [
   OnPaint!() -> (then, Context:"Paint Context Structure (by ref)") {is=title K2Node_FunctionEntry}
   > Return Node!(execute)
   n1 = Length(TargetArray=Get data())
   Branch(Condition="integer > integer"(A=@n1, B=1))
   > Set l_count(l_count=int - int(A=@n1, B=1)) -> (Output_Get=0)
   n2 = GetCachedGeometry()
   > Set l_size(l_size=vector2d - vector2d(A=GetLocalSize!(Geometry=@n2) -> (ReturnValue:Vector 2D Structure), B=Get drawOffset()))
   > n3 = Sequence()
   > n4 = For Each Loop(Array=Get data())
   > n5 = Set l_curIndex(l_curIndex=@n4.Array Index) -> (Output_Get=0)
   > n6 = getDrawColorFromIndex(index=@n5)
   > Set l_drawColor(l_drawColor=@n6)
   > Branch(Condition="integer > integer"(A=Get l_curIndex() -> (l_curIndex=0), B=0))
   n7 = "To Float (Integer)"(InInt=Get l_count() -> (l_count=0))
   n8 = Get l_curIndex() -> (l_curIndex=0)
   n9 = "To Float (Integer)"(InInt=@n8)
   n10 = Get data()
   n11 = vector2d * vector2d(A=MakeVector2D(X=float / float(A=@n9, B=@n7), Y="Get (a copy)"(Array=@n10, Dimension 1=@n8)), B=Get l_size())
   n12 = int - int(A=Get l_curIndex() -> (l_curIndex=0), B=1)
   n13 = float / float(A="To Float (Integer)"(InInt=@n12), B="To Float (Integer)"(InInt=Get l_count() -> (l_count=0)))
   n14 = "Get (a copy)"(Array=Get data(), Dimension 1=@n12)
   n15 = Get Context() -> (Context:"Paint Context Structure (by ref)")
   > DrawLine!(
    execute
    Context:"Paint Context Structure (by ref)"=@n15
    PositionA:Vector 2D Structure=vector2d * vector2d(A=MakeVector2D(X=@n13, Y=@n14), B=Get l_size())
    PositionB:Vector 2D Structure=@n11
    Tint:Linear Color Structure=Get l_drawColor()
    bAntiAlias:bool=true
    Thickness:float=1
   ) -> (then)
   n16 = MakeVector2D(X=3, Y=3)
   > DrawBox!(
    execute
    Context:"Paint Context Structure (by ref)"=@n15
    Position:Vector 2D Structure=vector2d - vector2d(A=@n11, B=vector2d / float(A=@n16, B=2))
    Size:Vector 2D Structure=@n16
    Brush:Slate Brush Asset Object Reference=&/PCPT_TimellineDotBrush
    Tint:Linear Color Structure=Get l_drawColor()
   ) -> (then)
   Branch(execute=@n3.then_1, Condition="Not Equal (Float)"(A=Get currentTheshold(), B=0))
   n17 = Get data_Highest()
   n18 = Get data_Lowest()
   n19 = Get currentTheshold()
   n20 = Get()
   n21 = Lerp(A=@n20.drawOffset_Y, B=float - float(B=@n20.drawOffset_Y), Alpha=float - float(B=NormalizeToRange(Value=@n19, RangeMin=@n18, RangeMax=@n17)))
   n22 = Get!() -> (l_size_X:double, l_size_Y:double)
   n23 = Get()
   n24 = float + float(
    A=float * float(A=Get!() -> (drawOffset_X:double, drawOffset_Y:double).drawOffset_Y, B=0.5)
    B="Clamp (Float)"(Value=float * float(A=@n22.l_size_Y, B=@n21), Min=0, Max=@n23.l_size_Y)
   )
   > DrawLine!(
    execute
    Context=Get Context()
    PositionA_X:double=0
    PositionA_Y:double=@n24
    PositionB_X:double=@n23
    PositionB_Y:double=@n24
    Tint=(R=0.5625,G=0.5625,B=0.5625,A=1)
    bAntiAlias=true
    Thickness=1
   ) -> (then)
  ]}
 OnTickValidateVisuals {nodes [
   OnTickValidateVisuals() {is=title K2Node_FunctionEntry}
   > Branch(Condition=Get validateNextTick() -> (validateNextTick:bool))
   > Set validateNextTick(validateNextTick=false)
   > validateValueVisuals()
  ]}
 ValidateRT {nodes [
   ValidateRT() {is=title K2Node_FunctionEntry}
   > n1 = Sequence()
   n2 = GetLocalSize!(Geometry=GetCachedGeometry()) -> (ReturnValue_X:double, ReturnValue_Y:double)
   n3 = Truncate(A=@n2)
   n4 = Truncate(A=@n2.ReturnValue_Y)
   > Set!(execute, CurrentRTSize_X:double="To Float (Integer)"(InInt=@n3), CurrentRTSize_Y:double="To Float (Integer)"(InInt=@n4)) -> (then, Output_Get:Vector 2D Structure)
   > n5 = Is Valid(InputObject=Get CurrentRT())
   n6 = CreateRenderTarget2D!(
    execute=@n5.Is Not Valid
    Width:int=@n3
    Height:int=@n4
    Format:ETextureRenderTargetFormat Enum=RTF_RGBA8
    ClearColor:Linear Color Structure=(R=0,G=0,B=0,A=0)
    bAutoGenerateMipMaps:bool=false
    bSupportUAVs:bool=false
   ) -> (then, ReturnValue:Texture Render Target 2D Object Reference)
   > n7 = Set CurrentRT(CurrentRT=@n6)
   > n8 = GetTimelineMaterial!(execute, self) -> (then, ReturnValue:Material Instance Dynamic Object Reference) {is=title K2Node_CallFunction}
   > SetTextureParameterValue(self=@n8, ParameterName=Texture, Value=@n7)
   > PrintString!(execute, InString=Append!(A="New RT ", B=GetDisplayName(Object=@n7)) -> (ReturnValue), bPrintToScreen, bPrintToLog, TextColor, Duration, Key) -> (then)
   n9 = Get!() -> (CurrentRTSize_X:double, CurrentRTSize_Y:double)
   ResizeRenderTarget2D!(
    execute=@n5.Is Valid
    TextureRenderTarget:Texture Render Target 2D Object Reference=Get CurrentRT()
    Width:int=Truncate(A=@n9)
    Height:int=Truncate(A=@n9.CurrentRTSize_Y)
   ) -> (then)
   > ClearRenderTarget2D!(execute, TextureRenderTarget:Texture Render Target 2D Object Reference=Get CurrentRT(), ClearColor:Linear Color Structure=(R=0,G=0,B=0,A=0)) -> (then)
   Set Timer for Next Tick by Event!(execute=@n1.then_1, Delegate:Delegate=Create Event()) -> (then, ReturnValue:Timer Handle Structure) {is=title K2Node_CallFunction}
  ]}
 average {nodes [
   n1 = average!() -> (then, InputPin:double=~) {is=title K2Node_FunctionEntry}
   > Return Node!(execute, OutputPin:double=@n1)
  ]}
 calculateValueFromMouseData {nodes [
   calculateValueFromMouseData() {is=title K2Node_FunctionEntry}
   > n1 = GetMousePositionOnPlatform()
   n2 = GetCachedGeometry()
   n3 = GetLocalSize(Geometry=@n2)
   n4 = "Absolute to Local (Coordinates)"(Geometry=@n2, AbsoluteCoordinate=@n1)
   > setValue(value="Clamp (Float)"(Value=float / float(A=@n4, B=@n3), Min=0, Max=1), throwEvent=true)
  ]}
 getDrawColorFromIndex {nodes [
   getDrawColorFromIndex!() -> (then, index:int) {is=title K2Node_FunctionEntry}
   > n1 = Branch(Condition=Is Valid Index!(
     TargetArray:Array of Linear Color Structures=Get dataColors() -> (dataColors:Array of Linear Color Structures)
     IndexToTest:int=Get index() -> (index:int=0)
    ) -> (ReturnValue:bool))
   > Return Node!(
    execute
    ReturnValue:Linear Color Structure="Get (a ref)"(Array:Array of Linear Color Structures=Get dataColors(), Dimension 1=Get index() -> (index=0)) -> (Output:"Linear Color Structure (by ref)")
   )
   Return Node(execute=@n1.else, ReturnValue=Get drawColor() -> (drawColor:Linear Color Structure))
  ]}
 getValue {nodes [
   getValue() {is=title K2Node_FunctionEntry}
   > Return Node!(execute, value:double=Get value())
  ]}
 handleMouseMove {nodes [
   handleMouseMove!() -> (then, Geometry:"Geometry Structure (by ref)", Input:"Pointer Event Structure (by ref)") {is=title K2Node_FunctionEntry}
   > calculateValueFromMouseData() {is=title K2Node_CallFunction}
  ]}
 normalizeGraphData {nodes [
   n1 = normalizeGraphData!() -> (then, data:"Array of Float (double-precision)s") {is=title K2Node_FunctionEntry}
   > Set l_tempData(l_tempData:"Array of Float (double-precision)s"=@n1) -> (Output_Get:"Array of Float (double-precision)s")
   n2 = Get l_tempData() -> (l_tempData:"Array of Float (double-precision)s")
   > Set data_Lowest(data_Lowest:double=MinOfFloatArray!(FloatArray:"Array of Float (single-precision)s"=@n2) -> (IndexOfMinValue:int=0, MinValue:float).MinValue) -> (Output_Get:double)
   > Set data_Highest(data_Highest:double=MaxOfFloatArray!(FloatArray:"Array of Float (single-precision)s"=@n2) -> (IndexOfMaxValue:int=0, MaxValue:float).MaxValue) -> (Output_Get:double)
   > n3 = Sequence()
   > n4 = For Each Loop(Array=Get l_tempData())
   n5 = NormalizeToRange(Value=@n4, RangeMin=Get data_Lowest(), RangeMax=Get data_Highest())
   n6 = Get!() -> (drawOffset_X:double, drawOffset_Y:double)
   > Add(
    TargetArray:"Array of Float (double-precision)s"=Get l_final() -> (l_final:"Array of Float (double-precision)s")
    NewItem:"Float (double-precision) (by ref)"=Lerp(A=@n6.drawOffset_Y, B=float - float(B=@n6.drawOffset_Y), Alpha=float - float(B=@n5))
   ) {is=title K2Node_CallArrayFunction}
   Return Node!(execute=@n3.then_1, normalizeData:"Array of Float (double-precision)s"=Get l_final())
  ]}
 onValueChanged {nodes [onValueChanged!() -> (then, value:double=~) {is=title K2Node_FunctionEntry}]}
 setGraphData {nodes [
   n1 = setGraphData!() -> (then, data:"Array of Float (double-precision)s", dataColors:Array of Linear Color Structures) {is=title K2Node_FunctionEntry}
   > Set InputData(InputData:"Array of Float (double-precision)s"=@n1) -> (Output_Get:"Array of Float (double-precision)s")
   > Set l_inputData(l_inputData:"Array of Float (double-precision)s"=@n1) -> (Output_Get:"Array of Float (double-precision)s")
   > Set dataColors(dataColors:Array of Linear Color Structures=@n1.dataColors) -> (Output_Get:Array of Linear Color Structures)
   > Clear(TargetArray:"Array of Float (double-precision)s"=Get data())
   n2 = Get l_inputData() -> (l_inputData:"Array of Float (double-precision)s")
   > Branch(Condition="integer > integer"(A=Length(TargetArray=@n2), B=1))
   > n3 = Sequence!(execute) -> (then_0, then_1, then_2:Exec)
   > Branch(Condition="integer > integer"(A=Get averageOver() -> (averageOver:int=0), B=0))
   > Set l_tempData(l_tempData=Get l_inputData())
   > Clear(TargetArray=Get l_inputData())
   > n4 = For Each Loop(Array=Get l_tempData())
   > Set l_curAverageIndex(l_curAverageIndex:int=@n4.Array Index) -> (Output_Get:int=0)
   > Clear(TargetArray=Get l_averageArray() -> (l_averageArray:"Array of Float (double-precision)s"))
   n5 = Get averageOver() -> (averageOver=0)
   n6 = Get l_curAverageIndex() -> (l_curAverageIndex:int=0)
   n7 = Get l_tempData()
   n8 = Get averageOver() -> (averageOver=0)
   n9 = Get l_curAverageIndex() -> (l_curAverageIndex=0)
   > n10 = For Loop(
    FirstIndex="Max (Integer)"(A=int - int(A=@n6, B=@n5), B=0)
    LastIndex="Min (Integer)"(
     A=int + int(A=@n9, B=@n8) -> (ReturnValue=0) {is=title K2Node_CommutativeAssociativeBinaryOperator}
     B=Last Index(TargetArray=@n7) -> (ReturnValue=0)
    ) -> (ReturnValue=0)
   )
   Set l_val(execute=@n10.Completed, l_val:double=0) -> (Output_Get:double)
   > n11 = For Each Loop(Array=Get l_averageArray())
   n12 = "To Float (Integer)"(InInt=Length(TargetArray=Get l_averageArray()))
   Add(execute=@n11.Completed, TargetArray=Get l_inputData(), NewItem=float / float(A=Get l_val() -> (l_val:double), B="Max (Float)"(A=@n12, B=1)))
   Set l_val(execute=@n11.LoopBody, l_val=float + float(A=@n11, B=Get l_val()))
   Add(execute=@n10.LoopBody, TargetArray=Get l_averageArray(), NewItem="Get (a copy)"(Array=Get l_tempData(), Dimension 1=@n10))
   n13 = normalizeGraphData!(execute=@n3.then_1, self, data:"Array of Float (double-precision)s"=Get l_inputData()) -> (then, normalizeData:"Array of Float (double-precision)s") {is=title K2Node_CallFunction}
   > Set data(data:"Array of Float (double-precision)s"=@n13) -> (Output_Get:"Array of Float (double-precision)s")
   Set Timer for Next Tick by Event(execute=@n3.then_2, Delegate=Create Event())
  ]}
 setValue {nodes [
   n1 = setValue!() -> (then, value:double=~, throwEvent:bool=true) {is=title K2Node_FunctionEntry}
   > Set value(value:double="Clamp (Float)"(Value=@n1, Min=0, Max=1)) -> (Output_Get:double)
   > validateValueVisuals()
   > Branch(Condition=@n1.throwEvent)
   > Call On Value Changed!(execute:Exec, self:PCPT Timeline Widget Object Reference, value:double=Get value()) -> (then:Exec) {is=title K2Node_CallDelegate}
  ]}
 validateValueVisuals {nodes [
   validateValueVisuals() {is=title K2Node_FunctionEntry}
   > n1 = GetTimelineMaterial()
   > SetScalarParameterValue(self=@n1, ParameterName=percent, Value=Get value())
  ]}
]
variables [
 averageOver {type=integer}
 CurrentRT {type=TextureRenderTarget2D, default=null}
 CurrentRTSize {type=Vector2D, default=<0 0>}
 currentTheshold {type=number}
 data {type="array<number>", default []}
 data_Highest {type=number}
 data_Lowest {type=number}
 dataColors {type="array<LinearColor>", default [<rgba 0 0 0 0>]}
 DrawAttempts {type=integer}
 drawColor {type=LinearColor, default=<rgba 0.5469 0.1082 0.0537 1>}
 drawOffset {type=Vector2D, default=<0 0>}
 drawOffset_0 {type=Vector2D, default=<0.015 0.1>}
 InputData {type="array<number>", default []}
 isDragging {type=boolean}
 isMouseOver {type=boolean}
 LastDrawSize {type=Vector2D, default=<0 0>}
 onValueChanged {}
 TimerHandle_DrawRTNextTick {type=TimerHandle}
 TimerHandle_ValidateRT {type=TimerHandle}
 validateNextTick {type=boolean}
 value {type=number, default=0.5}
]
widgets {
 info {inheritedWidgetCount=0, parentClass=&UserWidget, rootWidgetClass=&Border, widgetCount=2}
 widgets [
  Border_108 {
   widgetClassPath=&Border
   bIsVariable=true
   widget=Border_108
   properties {brushColor=<rgba 0.0104 0.0104 0.0104 1>, padding=<ltrb 0 15 0 15>}
  }
  Image_80 {
   widgetClassPath=&Image
   bIsVariable=true
   parent=Border_108
   widget=Image_80
   properties {brush {
     bIsDynamicallyLoaded=false
     drawAs=Image
     imageType=FullColor
     mirroring=NoMirror
     resourceName=None
     resourceObject=&/PCPT_Timeline_UIMat
     tiling=NoTile
     imageSize=<32 32>
     margin=<ltrb 0 0 0 0>
     outlineSettings {
      bUseBrushTransparency=false
      roundingType=HalfHeightRadius
      width=0
      color {colorUseRule=UseColor_Specified, specifiedColor=<rgba 0 0 0 0>}
      cornerRadii=<0 0 0 0>
     }
     tintColor {colorUseRule=UseColor_Specified, specifiedColor=<rgba 1 1 1 1>}
     uVRegion {bIsValid=false, max=<0 0>, min=<0 0>}
    }}
   slot {class=BorderSlot, properties {padding=<ltrb 0 15 0 15>}}
  }
 ]
}
