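# Editor utility subwidget for the perf CSV playback tool that wraps PCPT_Timeline_Widget with header, min/max and current-position labels.
# setGraphData feeds the timeline, labels its lowest and highest values, sets X range to the array indices and updates the header, hidden when empty.
# SetX normalises a value into that range and drives the timeline.
# Timeline changes convert percent back to X, optionally rounded, and slide the current label via left padding.
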WidgetBlueprint /Game/NeverCook/EditorUtilities/PerfTools/PerfCsvPlaybackTool/SubWidgets/PCPT_TimelineMarkedUp
graphs [
 EventGraph {nodes [
   n1 = "On Value Changed (timeline)"!() -> (OutputDelegate, then, value:double) {is=title K2Node_ComponentBoundEvent}
   > "SetText (Text)"(
    self=Get txtCurrent() -> (txtCurrent:Common Text Object Reference)
    InText=Format Text!(Format="LOCTEXT(\"{time}\")", time:double=GetXFromPercent!(Percent:double=@n1.value) -> (X:double)) -> (Result)
   )
   n2 = GetAbsoluteSize!(Geometry:"Geometry Structure (by ref)"=GetCachedGeometry!(self:Widget Object Reference=Get timeline() -> (timeline:PCPT Timeline Widget Object Reference)) -> (ReturnValue:"Geometry Structure (by ref)")) -> (ReturnValue_X:double, ReturnValue_Y:double)
   > SetPadding!(
    execute
    self:Vertical Box Slot Object Reference=SlotAsVerticalBoxSlot(Widget=Get txtCurrent())
    InPadding:Margin Structure=Make Margin(Left=Lerp(A=0, B=@n2, Alpha=@n1.value), Top=0, Right=0, Bottom=0)
   ) -> (then)
  ]}
 GetXFromPercent {nodes [
   n1 = GetXFromPercent!() -> (then, Percent:double=~) {is=title K2Node_FunctionEntry}
   > n2 = Branch(Condition=Get TreatXAsInteger() -> (TreatXAsInteger:bool))
   n3 = Lerp(A=Get MinX() -> (MinX:double), B=Get MaxX() -> (MaxX:double), Alpha=@n1)
   > Return Node!(execute, X:double="To Float (Integer)"(InInt=Round(A=@n3) {is=title K2Node_CallFunction}))
   Return Node(execute=@n2.else, X=@n3)
  ]}
 SetHeader {nodes [
   n1 = SetHeader!() -> (then, InText:Text) {is=title K2Node_FunctionEntry}
   > "SetText (Text)"(self=Get txtHeader() -> (txtHeader:Common Text Object Reference), InText=@n1)
   > SetVisibility(self=Get txtHeader(), InVisibility=Select(Option 0=SelfHitTestInvisible, Option 1=Collapsed, Index=TextIsEmpty(InText=@n1)))
  ]}
 SetMinMaxX {nodes [
   n1 = SetMinMaxX!() -> (then, MinX:double=~, MaxX:double=~) {is=title K2Node_FunctionEntry}
   > Set MinX(MinX:double=@n1) -> (Output_Get:double)
   > Set MaxX(MaxX:double=@n1.MaxX) -> (Output_Get:double)
  ]}
 SetX {nodes [
   n1 = SetX!() -> (then, Value:double=~) {is=title K2Node_FunctionEntry}
   > setValue!(
    execute
    self:PCPT Timeline Widget Object Reference=Get timeline()
    value:double=NormalizeToRange!(Value:double=@n1, RangeMin:double=Get MinX(), RangeMax:double=Get MaxX()) -> (ReturnValue:double)
    throwEvent:bool=true
   ) -> (then)
  ]}
 setGraphData {nodes [
   n1 = setGraphData!() -> (then, data:"Array of Float (double-precision)s", dataColors:Array of Linear Color Structures, headerLabel:Text) {is=title K2Node_FunctionEntry}
   > setGraphData!(
    execute
    self:PCPT Timeline Widget Object Reference=Get timeline()
    data:"Array of Float (double-precision)s"=@n1
    dataColors:Array of Linear Color Structures=@n1.dataColors
   ) -> (then) {is=title K2Node_CallFunction}
   n2 = Get timeline()
   > "SetText (Text)"(
    self=Get txtMin() -> (txtMin:Common Text Object Reference)
    InText="To Text (Float)"(
     Value=Get data_Lowest!(self:PCPT Timeline Widget Object Reference=@n2) -> (data_Lowest:double)
     MinimumIntegralDigits=1
     MaximumIntegralDigits=324
     MinimumFractionalDigits=0
     MaximumFractionalDigits=2
    )
   )
   n3 = Get timeline()
   > "SetText (Text)"(
    self=Get txtMax() -> (txtMax:Common Text Object Reference)
    InText="To Text (Float)"(
     Value=Get data_Highest!(self:PCPT Timeline Widget Object Reference=@n3) -> (data_Highest:double)
     MinimumIntegralDigits=1
     MaximumIntegralDigits=324
     MinimumFractionalDigits=0
     MaximumFractionalDigits=2
    )
   )
   > SetMinMaxX!(
    execute
    self
    MinX:double=0
    MaxX:double="To Float (Integer)"(InInt=Last Index!(TargetArray:"Array of Float (double-precision)s"=@n1) -> (ReturnValue=0))
   ) -> (then) {is=title K2Node_CallFunction}
   > SetHeader!(execute, self, InText:Text=Get headerLabel() -> (headerLabel:Text)) -> (then) {is=title K2Node_CallFunction}
  ]}
]
variables [
 MaxX {type=number, default=1}
 MinX {type=number}
 TreatXAsInteger {type=boolean}
]
widgets {
 info {inheritedWidgetCount=0, parentClass=&UserWidget, rootWidgetClass=&SizeBox, widgetCount=11}
 widgets [
  SizeBox_0 {
   widgetClassPath=&SizeBox
   widget=SizeBox_0
   properties {bOverride_MinDesiredHeight=true, heightOverride=250, minDesiredHeight=250}
  }
  VerticalBox_0 {
   widgetClassPath=&VerticalBox
   parent=SizeBox_0
   widget=VerticalBox_0
   slot {class=SizeBoxSlot, properties {padding=<ltrb 10 0 10 0>}}
  }
  HorizontalBox_562 {widgetClassPath=&HorizontalBox, parent=VerticalBox_0, widget=HorizontalBox_562, slot {class=VerticalBoxSlot}}
  VerticalBox_4 {
   widgetClassPath=&VerticalBox
   parent=HorizontalBox_562
   widget=VerticalBox_4
   slot {class=HorizontalBoxSlot, properties {horizontalAlignment=HAlign_Right, padding=<ltrb 0 16 0 15>}}
  }
  txtMax {
   widgetClassPath=&CommonTextBlock
   bIsVariable=true
   parent=VerticalBox_4
   widget=txtMax
   properties {
    justification=Right
    text=Max
    font {
     bForceMonospaced=false
     bMaterialIsStencil=false
     fontMaterial=null
     fontObject=&NotoSans
     letterSpacing=0
     monospacedWidth=1
     size=8
     skewAmount=0
     typefaceFontName=NotoSans-Regular
     outlineSettings {
      bApplyOutlineToDropShadows=false
      bMiteredCorners=false
      bSeparateFillAlpha=false
      outlineMaterial=null
      outlineSize=0
      outlineColor=<rgba 0 0 0 1>
     }
    }
    shadowOffset=<0 0>
   }
   slot {class=VerticalBoxSlot, properties {verticalAlignment=VAlign_Top, size {sizeRule=Fill, value=1}}}
  }
  txtMin {
   widgetClassPath=&CommonTextBlock
   bIsVariable=true
   parent=VerticalBox_4
   widget=txtMin
   properties {
    justification=Right
    text=Min
    font {
     bForceMonospaced=false
     bMaterialIsStencil=false
     fontMaterial=null
     fontObject=&NotoSans
     letterSpacing=0
     monospacedWidth=1
     size=8
     skewAmount=0
     typefaceFontName=NotoSans-Regular
     outlineSettings {
      bApplyOutlineToDropShadows=false
      bMiteredCorners=false
      bSeparateFillAlpha=false
      outlineMaterial=null
      outlineSize=0
      outlineColor=<rgba 0 0 0 1>
     }
    }
    shadowOffset=<0 0>
   }
   slot {class=VerticalBoxSlot, properties {verticalAlignment=VAlign_Bottom, size {sizeRule=Fill, value=1}}}
  }
  VerticalBox_1 {
   widgetClassPath=&VerticalBox
   parent=HorizontalBox_562
   widget=VerticalBox_1
   slot {class=HorizontalBoxSlot, properties {size {sizeRule=Fill, value=1}}}
  }
  txtHeader {
   widgetClassPath=&CommonTextBlock
   bIsVariable=true
   parent=VerticalBox_1
   widget=txtHeader
   properties {
    justification=Right
    text=Current
    visibility=Collapsed
    font {
     bForceMonospaced=false
     bMaterialIsStencil=false
     fontMaterial=null
     fontObject=&NotoSans
     letterSpacing=0
     monospacedWidth=1
     size=8
     skewAmount=0
     typefaceFontName=NotoSans-Regular
     outlineSettings {
      bApplyOutlineToDropShadows=false
      bMiteredCorners=false
      bSeparateFillAlpha=false
      outlineMaterial=null
      outlineSize=0
      outlineColor=<rgba 0 0 0 1>
     }
    }
    shadowOffset=<0 0>
   }
   slot {class=VerticalBoxSlot, properties {horizontalAlignment=HAlign_Left, verticalAlignment=VAlign_Bottom}}
  }
  SizeBox {
   widgetClassPath=&SizeBox
   parent=VerticalBox_1
   widget=SizeBox
   properties {bOverride_HeightOverride=true, heightOverride=207.307}
   slot {class=VerticalBoxSlot}
  }
  timeline {
   widgetClassPath=&/PCPT_Timeline_Widget
   bIsVariable=true
   parent=SizeBox
   widget=timeline
   properties {averageOver=5, value=0}
   slot {class=SizeBoxSlot}
  }
  txtCurrent {
   widgetClassPath=&CommonTextBlock
   bIsVariable=true
   parent=VerticalBox_1
   widget=txtCurrent
   properties {
    justification=Right
    text=Current
    font {
     bForceMonospaced=false
     bMaterialIsStencil=false
     fontMaterial=null
     fontObject=&NotoSans
     letterSpacing=0
     monospacedWidth=1
     size=8
     skewAmount=0
     typefaceFontName=NotoSans-Regular
     outlineSettings {
      bApplyOutlineToDropShadows=false
      bMiteredCorners=false
      bSeparateFillAlpha=false
      outlineMaterial=null
      outlineSize=0
      outlineColor=<rgba 0 0 0 1>
     }
    }
    shadowOffset=<0 0>
   }
   slot {class=VerticalBoxSlot, properties {horizontalAlignment=HAlign_Left, verticalAlignment=VAlign_Bottom}}
  }
 ]
}
